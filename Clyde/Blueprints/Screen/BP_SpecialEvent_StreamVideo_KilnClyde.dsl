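; Special event stream-video screen for KilnClyde that plays a streamed video and reports viewer attendance.
; Before parent BeginPlay, it sets the media source's stream IDs from a hotfixable DataTable row.
; It then registers with the movie-set camera and audio analysis subsystem.
; On cosmetic-capable instances, media open starts timing; media end, stream close or EndPlay fires the attendance event once, guarded by a tracking flag.
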(event EventBeginPlay
  ;[1 1 0 0.5] Must be called before BeginPlay in order to set the VUID
  (CallFunction|SetStreamURL)
  (|CallParentFunction)
  (CallFunction|RegisterwithMovieCamera)
  (AudioAnalysis|RegisterDefaultSubmix (WorldSubsystems|GetFortAudioAnalysisSubsystem))
  (Media|MediaPlayer|BindEventtoOnEndReached (Variables|Default|GetVideoPlayer) (EventDispatchers|CreateEvent :!event Custom|OnMediaEnded)))

(event EventEndPlay
  (params (EndPlayReason :type EEndPlayReason%20Enum :default "Destroyed"))
  (|CallParentFunction EndPlayReason)
  (AudioAnalysis|UnRegisterDefaultSubmix (WorldSubsystems|GetFortAudioAnalysisSubsystem))
  (CallFunction|RecordPlaybackAnalytics))

(event EventOpenSpecialEventStream
  (|CallParentFunction))

(event EventCloseSpecialEventStream
  (|CallParentFunction)
  (CallFunction|RecordPlaybackAnalytics))

(event EventOnMediaOpened
  (params (OpenedUrl :type String))
  (|CallParentFunction OpenedUrl)
  (CallFunction|StartTrackingPlaybackAnalytics))

(event Custom|OnMediaEnded
  (CallFunction|RecordPlaybackAnalytics))

(event Custom|Set%20Stream%20URL
  (bind _mediasource (Variables|Default|GetMediaSource))
  ;[0 0 0 0.5] Hotfix Stream VUID
  (bind _returnvalue (Utilities|GetDataTableRow "/Clyde/Datatables/KilnClydeURL_Data.KilnClydeURL_Data" "KilnURL")
    (:then
      (bind _returnvalue_kilnurl_3_94d1c58b48066ad4557f9eb17250cc20 (Utilities|Struct|BreakKilnClydeURL _returnvalue))
      (bind _output_get (Variables|/Script/EpicStreamMediaSource.EpicStreamMediaSource|SetStreamID _returnvalue_kilnurl_3_94d1c58b48066ad4557f9eb17250cc20 _mediasource))
      (bind _output_get_2 (Variables|/Script/EpicStreamMediaSource.EpicStreamMediaSource|SetStreamID_Development _output_get _mediasource))
      (Fortnite|FortShippingLog self (Utilities|String|Append "Set stream VUID to: " _output_get_2)))
    (:RowNotFound
      (Fortnite|FortShippingLog self "Unable to find stream URL :( "))))

(fn RecordPlaybackAnalytics
  (Networking|CanExecuteCosmeticEvents
    (:True
      (if (Variables|Analytics|GetTrackingPlayback)
        (Analytics|FireEventClydeEventAttendance (Fortnite|GetFirstLocalFortPlayerController true) (Variables|Analytics|GetStartPlaybackDateTime) (Math|DateTime|UTCNow))
        (Variables|Analytics|SetTrackingPlayback false)))))

(fn RegisterWithMovieCamera
  (bind _returnvalue (Actor|GetActorOfClass "/Clyde/Blueprints/Screen/Clyde_Movie_Set_Camera.Clyde_Movie_Set_Camera_C"))
  (Utilities|IsValid _returnvalue
    (:Is%20Valid
      (bind _output_get (Variables|Default|SetMovieSetCamera _returnvalue))
      (Class|ClydeMovieSetCamera|RegisterVideoStream _output_get self))
    (:Is%20Not%20Valid
      (Fortnite|FortShippingLog self "Could not find Move Set Camera!" true))))

(fn StartTrackingPlaybackAnalytics
  (Networking|CanExecuteCosmeticEvents
    (:True
      (Variables|Analytics|SetTrackingPlayback true)
      (Variables|Analytics|SetStartPlaybackDateTime (Math|DateTime|UTCNow)))))

(fn ConstructionScript
  (|CallParentFunction))
