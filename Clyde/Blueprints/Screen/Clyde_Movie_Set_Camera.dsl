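; Interactable movie set camera that plays a streamed film once via a "Play Movie" prompt.
; Server interaction activates the screen data layer and broadcasts KilnFlick.Viewed quest verbs to the interactor and players inside the quest volume.
; The screen's registration starts the stream, projector effects and volume overlaps, and blocks further interaction; media end disables them and hides the screen.
; Construction rotates the camera and feeds projection endpoints into a dynamic projector-line material.
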(event EventBeginPlay
  ;[0 0 0 0.5] BeginPlay
  (Components|Activation|Deactivate (Variables|Default|GetNS_CameraScreenProjectorFX))
  (Rendering|SetVisibility (Variables|Default|GetProjector_Lines) false false))

(event Interaction|EventBlueprintOnInteract
  (params (InteractingPawn :type Fort%20Pawn%20Object%20Reference) (InteractionBeingAttempted :type EInteractionBeingAttempted%20Enum :default "FirstInteraction"))
  ;[0 0 0 0.5] BlueprintOnInteract: only runs on Server
  (bind _returnvalue (WorldPartition|GetDataLayerManager))
  (DataLayers|SetDataLayerRuntimeState _returnvalue "/Clyde/DataLayers/Clyde_S2/DL_CLY_StreamScreen.DL_CLY_StreamScreen" "Activated" false :!class /Script/Engine.DataLayerManager)
  (CustomVerb|BroadcastCustomVerbMessage (CustomVerb|MakeCustomVerbMessage InteractingPawn "(TagName=\"CustomVerb.KilnFlick.Viewed\")"))
  (bind _returnvalue_2 (Actor|GetActorOfClass "/Clyde/Assets/ScreenStream/KilnClyde_Volume.KilnClyde_Volume_C"))
  (for _array_element (Collision|GetOverlappingActors _returnvalue_2 "/Game/Athena/PlayerPawn_Athena.PlayerPawn_Athena_C")
    (CustomVerb|BroadcastCustomVerbMessage (CustomVerb|MakeCustomVerbMessage _array_element "(TagName=\"CustomVerb.KilnFlick.Viewed\")"))))

(event Custom|RegisterVideoStream
  (params (InVideoStream :type BP%20Special%20Event%20Stream%20Video%20Kiln%20Clyde%20Object%20Reference))
  ;[0 0 0 0.5] Wait for RR Screen to register itself before attempting to activate the stream
  (Utilities|IsValid InVideoStream
    (:Is%20Valid
      (bind _output_get (Variables|Default|SetVideo%20Stream InVideoStream))
      (CallFunction|ActivateStream _output_get))
    (:Is%20Not%20Valid
      (Fortnite|FortShippingLog self "Attempted to register NULL video stream BP!" true))))

(event Custom|ActivateStream
  (params (InVideoStream :type BP%20Special%20Event%20Stream%20Video%20Kiln%20Clyde%20Object%20Reference))
  ;[0 0 0 0.5] ActivateStream
  (Utilities|IsValid InVideoStream
    (:Is%20Valid
      ;; Unhide the stream screen, call to start the stream
      (Rendering|SetActorHiddenInGame InVideoStream false)
      (Rendering|SetVisibility (Variables|/Clyde/Blueprints/Screen/BP_SpecialEvent_StreamVideo_KilnClyde.BP_SpecialEvent_StreamVideo_KilnClyde_C|GetStreaming%20screen InVideoStream) true false)
      (Class|BPSpecialEventStreamVideoParent|OpenSpecialEventStream InVideoStream)
      (bind _returnvalue (Actor|GetActorOfClass "/Clyde/Blueprints/Screen/Clyde_Movie_Set_Camera.Clyde_Movie_Set_Camera_C"))
      (Effects|Components|ParticleSystem|SetEmitterEnable (Variables|Default|GetNS_CameraScreenProjectorFX _returnvalue) "None" true)
      (Components|Activation|Activate (Variables|Default|GetNS_CameraScreenProjectorFX _returnvalue) true)
      (Rendering|SetVisibility (Variables|Default|GetProjector_Lines) true false)
      (Variables|Default|SetisButtonPressed true)
      (bind _returnvalue_2 (Actor|GetActorOfClass "/Clyde/Assets/ScreenStream/KilnClyde_Volume.KilnClyde_Volume_C"))
      (PhysicsObject|EnableOverlapEvents _returnvalue_2)
      (Collision|SetCollisionEnabled (Variables|/Clyde/Assets/ScreenStream/KilnClyde_Volume.KilnClyde_Volume_C|GetQuestOverlapVolume _returnvalue_2) "QueryOnly")
      ;; Bind event to disable Quest volume when media is closed
      (Media|MediaPlayer|BindEventtoOnEndReached (Variables|/SpecialEventGameplay/Blueprints/Media/BP_SpecialEvent_StreamVideoParent.BP_SpecialEvent_StreamVideoParent_C|GetVideoPlayer (Variables|Default|GetVideo%20Stream)) (delegate Custom|OnMediaEndReached)))))

(event Custom|OnMediaEndReached
  ;[0 0 0 0.5] ActivateStream
  ;; Bind event to disable Quest volume when media is closed
  (CallFunction|DisableVolume))

(event Custom|DisableVolume
  (bind _video_stream (Variables|Default|GetVideo%20Stream))
  (bind _ns_camerascreenprojectorfx (Variables|Default|GetNS_CameraScreenProjectorFX))
  ;[0 0 0 0.5] Disable the Quest volume and hide the screen stream
  (bind _returnvalue (Actor|GetActorOfClass "/Clyde/Assets/ScreenStream/KilnClyde_Volume.KilnClyde_Volume_C"))
  (PhysicsObject|DisableOverlapEvents _returnvalue)
  (Collision|SetCollisionEnabled (Variables|/Clyde/Assets/ScreenStream/KilnClyde_Volume.KilnClyde_Volume_C|GetQuestOverlapVolume _returnvalue) "NoCollision")
  (Components|Activation|Deactivate _ns_camerascreenprojectorfx)
  (Effects|Components|ParticleSystem|SetEmitterEnable _ns_camerascreenprojectorfx "None" false)
  (Rendering|SetVisibility (Variables|Default|GetProjector_Lines) false false)
  (Rendering|SetVisibility (Variables|/Clyde/Blueprints/Screen/BP_SpecialEvent_StreamVideo_KilnClyde.BP_SpecialEvent_StreamVideo_KilnClyde_C|GetStreaming%20screen _video_stream) false false)
  (Rendering|SetActorHiddenInGame _video_stream true))

(fn BlueprintCanInteract
  (params (InteractingPawn :type Fort%20Pawn%20Object%20Reference) (InteractionBeingAttempted :type EInteractionBeingAttempted%20Enum :default "FirstInteraction") (InteractionType :type TInteractionType%20Enum :default "IT_NoInteraction"))
  (if (Variables|Default|GetisButtonPressed)
    (return false)
    (else
      (return true))))

(fn BlueprintGetInteractionString
  (params (InteractingPawn :type Fort%20Pawn%20Object%20Reference) (InteractionBeingAttempted :type EInteractionBeingAttempted%20Enum :default "FirstInteraction"))
  (return "NSLOCTEXT(\"[95959F421C877ED998E1A4F7E77B4628]\", \"FA3CE76D4F7D2D20156903920257E478\", \"Play Movie\")"))

(fn ConstructionScript
  (bind _projectormid (Variables|Default|GetProjectorMID))
  (bind _projectionstart_location (Math|Transform|BreakTransform (Variables|GetProjectionStart_Location_X)))
  (bind (_projectionstart_location_x _projectionstart_location_y _projectionstart_location_z) (Math|Vector|BreakVector _projectionstart_location))
  (bind _projectionend_location (Math|Transform|BreakTransform (Variables|GetProjectionEnd_Location_X)))
  (bind (_projectionend_location_x _projectionend_location_y _projectionend_location_z) (Math|Vector|BreakVector _projectionend_location))
  (Transformation|SetRelativeRotation (Variables|Default|GetMovieset_Camera) (Math|Rotator|MakeRotator 0.0 0.0 (Variables|Default|GetCameraRotation)) false false)
  (bind _returnvalue (Rendering|Material|CreateDynamicMaterialInstance (Variables|Default|GetProjector_Lines) 0 (Rendering|Material|GetMaterial (Variables|Default|GetProjector_Lines) 0) "None" :!class /Script/Engine.PrimitiveComponent))
  (bind _output_get (Variables|Default|SetProjectorMID _returnvalue))
  (Rendering|Material|SetMaterial (Variables|Default|GetProjector_Lines) 0 _output_get)
  (Rendering|Material|SetVectorParameterValue _projectormid "StartOffset" (Math|Color|MakeColor _projectionstart_location_x _projectionstart_location_y _projectionstart_location_z 0.0) :!class /Script/Engine.MaterialInstanceDynamic)
  (Rendering|Material|SetVectorParameterValue _projectormid "EndOffset" (Math|Color|MakeColor _projectionend_location_x _projectionend_location_y _projectionend_location_z 0.0) :!class /Script/Engine.MaterialInstanceDynamic))
