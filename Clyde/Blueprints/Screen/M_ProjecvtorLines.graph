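# Additive, unlit, two-sided material for projector beam lines that appear to cast Clyde's screen video.
# Vertex blue blends a collapse toward actor position and local start/end offsets to shape the beam geometry.
# Color samples the KilnClyde media texture, with optional luminance-based reduction and clamping, then emissive adjustments.
# Opacity combines procedural edge bands, camera fade, vertex-blue masking and platform-switched depth fade, alongside a view-angle mask branch.
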Material /Clyde/Blueprints/Screen/M_ProjecvtorLines
graph {
 properties {bAutomaticallySetUsageInEditor=false, blendMode=BLEND_Additive, shadingModel=MSM_Unlit, twoSided=true, usedShadingModels=Unlit}
 nodes [
  n1 = ActorPositionWS()
  n2 = WorldPosition()
  n3 = ScalarParameter() {defaultValue=0.5, parameterName=Collaps amount Start}
  n4 = ScalarParameter() {parameterName=End to start Amount}
  n5 = LinearInterpolate(A=ScalarParameter() {parameterName=Collaps amount End}, B=@n3, Alpha=@n4)
  n6 = VertexColor()
  n7 = Transform(None=VectorParameter() {parameterName=StartOffset, defaultValue=<rgba 500 0 0 0>}.RGB) {transformSourceType=TRANSFORMSOURCE_Local}
  n8 = Transform(None=VectorParameter() {parameterName=EndOffset, defaultValue=<rgba 500 0 0 0>}.RGB) {transformSourceType=TRANSFORMSOURCE_Local}
  Add(
   A=Multiply(A=Subtract(A=@n1, B=@n2.XYZ) {desc=Collaspe}, B=LinearInterpolate(A=@n3, B=@n5, Alpha=@n6.B))
   B=LinearInterpolate(A=@n7, B=LinearInterpolate(A=@n8, B=@n7, Alpha=@n4), Alpha=VertexColor().B)
  )
  Add()
  n9 = TextureSample() {texture=&/SpecialEventsMediaTexture_Video_KilnClyde}
  n10 = Divide(A=Add(A=Add(A=@n9.R, B=@n9.R), B=@n9.B)) {constB=3}
  n11 = Multiply(A=@n10, B=LinearInterpolate(Alpha=OneMinus(None=@n10)) {constA=0.5, constB=1.5})
  n12 = LinearInterpolate(
   A=@n9.RGB
   B=Multiply(A=@n9.RGB, B=Saturate(None=@n11))
   Alpha=ScalarParameter() {defaultValue=0.5, group=Color, parameterName=Use Color reduction Method}
  )
  n13 = ScalarParameter() {group=Color, parameterName=Clamp color Min}
  n14 = ScalarParameter() {defaultValue=1, group=Color, parameterName=Clamp color Max}
  MaterialFunctionCall(Emissive=NamedRerouteDeclaration(None=Clamp(None=@n12, Min=@n13, Max=@n14)) {name=Emissive, nodeColor=<rgba 1 0 0.7734 1>}) {materialFunction=&MF_FX_EmissiveAdjustments}
  n15 = Abs(None=DotProduct(
    A=NamedRerouteUsage()
    B=MaterialFunctionCall() {materialFunction=&CameraDirectionVector}.Result
   ))
  n16 = VectorParameter() {parameterName=Dot Masking options, defaultValue=<rgba 1 1 0 0>}
  n17 = Power(Base=OneMinus(None=OneMinus(None=@n15)), Exp=@n16.R)
  n18 = MaterialFunctionCall(In=Add(A=Multiply(A=@n17, B=@n16.G), B=@n16.B), Contrast=@n16.A) {materialFunction=&CheapContrast}
  n19 = Constant() {r=0.5}
  n20 = ScalarParameter() {defaultValue=0.5, parameterName=Edge Band Sharpness}
  n21 = MaterialFunctionCall!(Width=@n19, Sharpness=@n20, Offset, Direction Switch, Compare, Input Coordinates) {materialFunction=/Engine/Functions/Engine_MaterialFunctions03/Procedurals/GeneratedBand}
  n22 = MaterialFunctionCall(Width=@n19, Sharpness=@n20, Direction Switch=StaticBool() {value=true}) {materialFunction=&GeneratedBand}
  n23 = Multiply(
   A=Power(Base=Min(A=@n21.Result, B=@n22.Result), Exp=ScalarParameter() {defaultValue=1.5, parameterName=Edge band Pow})
   B=MaterialFunctionCall(Fade Length=ScalarParameter() {parameterName=Camera depth faed length}, Fade Offset=ScalarParameter() {parameterName=Camera Fade Offset}) {materialFunction=&CameraDepthFade}.Result
  )
  n24 = DepthFade(Opacity=@n23, FadeDistance=ScalarParameter() {parameterName=Depth fade dist})
  n25 = VectorParameter() {parameterName=Masking options, defaultValue=<rgba 1 1 0 0>}
  n26 = Power(Base=OneMinus(None=VertexColor().B), Exp=@n25.R)
  n27 = MaterialFunctionCall(In=Add(A=Multiply(A=@n26, B=@n25.G), B=@n25.B), Contrast=@n25.A) {materialFunction=&CheapContrast}
  n28 = Multiply(
   A=Multiply(
    A=MaterialFunctionCall("Xb1,PS4"=@n24, "PC-High&&Epic,XSX,PS5"=@n24, "PC-Perf,Android"=@n23, Switch=@n23, "PC-Low(NotPerf)&&Medium"=@n24) {materialFunction=&PlatformSwitch_1D}.Result
    B=Saturate(None=@n27.Result)
   )
   B=ScalarParameter() {defaultValue=1, parameterName=Master Opacity}
  )
  Multiply(A=OneMinus(None=Saturate(None=@n18.Result)), B=@n28)
  NamedRerouteDeclaration(None=@n28) {name=Opac, nodeColor=<rgba 0 0.1328 1 1>}
  NamedRerouteDeclaration(None=Normalize(VectorInput=Subtract(A=@n7, B=@n8))) {name=Offset direction, nodeColor=<rgba 0.2656 0 1 1>}
  VertexColor()
 ]
}
