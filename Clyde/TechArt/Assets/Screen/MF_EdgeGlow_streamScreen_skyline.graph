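# Material function for the skyline stream screen's edge glow, outputting emissive color and opacity from base color, audio and video-switch inputs.
# A screen SDF and crack texture form masks, with audio driving SDF falloff and edge intensity.
# Panned radial cloud, flow and star textures supply animated glow, distortion-like detail and light rays.
# Glow blends with base color through emissive adjustments, while the video switch selects between opacity masks.
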MaterialFunction /Clyde/TechArt/Assets/Screen/MF_EdgeGlow_streamScreen_skyline
graph {nodes [
  n1 = NamedRerouteUsage()
  n2 = NamedRerouteUsage()
  Add(A=@n1, B=@n2)
  Constant()
  n3 = NamedRerouteUsage()
  n4 = NamedRerouteUsage()
  n5 = NamedRerouteUsage()
  n6 = Subtract(A=@n2, B=@n1)
  n7 = LinearInterpolate(
   A=Multiply(A=NamedRerouteDeclaration(None=@n3) {name=Opacity_1, nodeColor=<rgba 1 0 0.2812 1>}, B=Add(A=@n4, B=@n5))
   B=FunctionInput(Preview=Constant3Vector()) {inputName=BaseColor}
   Alpha=Saturate(None=Saturate(None=@n6))
  )
  FunctionOutput(None=MaterialFunctionCall(Emissive=NamedRerouteDeclaration(None=@n7) {name=EmissiveColor, nodeColor=<rgba 1 0.6328 0 1>}) {materialFunction=&MF_FX_EmissiveAdjustments}.Result) {outputName=EmissiveColor}
  n8 = NamedRerouteUsage()
  n9 = NamedRerouteUsage()
  n10 = NamedRerouteUsage()
  n11 = Max(A=@n8, B=Max(A=Power(Base=@n9), B=Saturate(None=@n10)))
  n12 = NamedRerouteDeclaration(None=FunctionInput() {bUsePreviewValueAsDefault=true, inputName=Video switch, inputType=FunctionInput_Scalar}) {name=VideoSwitch, nodeColor=<rgba 0.2656 0 1 1>}
  FunctionOutput(None=LinearInterpolate(A=@n8, B=NamedRerouteDeclaration(None=@n11) {name=Opacity, nodeColor=<rgba 0 1 0.9531 1>}, Alpha=@n12)) {outputName=Opacity}
  n13 = ScalarParameter() {defaultValue=0.4589, parameterName=Paramd}
  n14 = ScalarParameter() {defaultValue=0.716, parameterName=Paramfff}
  n15 = NamedRerouteUsage()
  n16 = Constant()
  n17 = Constant() {r=1}
  n18 = Constant() {r=0.083}
  n19 = NamedRerouteUsage()
  n20 = Constant()
  n21 = Saturate(None=Multiply(
    A=MaterialFunctionCall(A=@n16, B=@n17, Transition Phase=@n18, Height Texture=@n19, Contrast=@n20) {materialFunction=&HeightLerp}.Results
    B=Constant() {r=30}
   ))
  n22 = Subtract(A=NamedRerouteUsage(), B=@n21)
  Multiply(A=SmoothStep(Min=LinearInterpolate(A=@n13, B=@n14, Alpha=@n15) {constA=0.8}, Max=Constant() {r=1}, Value=@n22), B=@n15)
  n23 = TextureCoordinate()
  n24 = Constant2Vector()
  n25 = Add(
   A=MaterialFunctionCall(UVs=Add(A=@n23, B=@n24), Texture Scale=ScalarParameter() {defaultValue=0.95, parameterName=scale skyline}) {materialFunction=&ScaleUVsByCenter}.UVs
   B=Constant()
  )
  n26 = TextureSample(UVs=@n25) {samplerType=SAMPLERTYPE_LinearColor, texture=&/T_Screen_crack}
  Multiply(A=@n26.R) {constB=0.454}
  n27 = TextureCoordinate()
  n28 = Constant2Vector() {g=1, r=1}
  n29 = Constant() {r=3}
  n30 = Add(A=Multiply(A=Multiply(A=@n27, B=@n28), B=@n29) {constB=3}, B=Multiply(A=@n29) {constB=-0.5}) {constB=-1.5}
  n31 = TextureCoordinate() {coordinateIndex=1}
  n32 = TextureSample!(
   UVs=Panner(Coordinate=MaterialFunctionCall(Vector or UVs=@n30) {materialFunction=&VectorToRadialValue}.Radial Coordinates) {speedX=0.1, speedY=-0.02}
   Tex
   "DDX(UVs)"=DDX(Value=@n31)
   "DDY(UVs)"=DDY(Value=@n31)
   Apply View MipBias
  ) {
   mipValueMode=TMVM_Derivative
   samplerSource=SSM_Wrap_WorldGroupSettings
   samplerType=SAMPLERTYPE_Normal
   texture=/Clyde/TechArt/Assets/mesh/SkyBox/T_Klin_Flow_03
  }
  Multiply(A=@n32.R, B=LinearInterpolate(B=Constant() {r=0.0015}, Alpha=@n12)) {constB=0.05}
  n33 = NamedRerouteUsage()
  n34 = Max(A=NamedRerouteUsage(), B=@n22)
  NamedRerouteDeclaration(None=Add(A=NamedRerouteUsage(), B=Multiply(A=@n33, B=@n34))) {name=EdgeMask, nodeColor=<rgba 1 0 0.5156 1>}
  n35 = TextureSample!(UVs=@n25, Tex, Apply View MipBias) {samplerType=SAMPLERTYPE_LinearColor, texture=&/T_Screen_Mask_SDF}
  n36 = Multiply(A=@n35.R) {constB=1.5}
  n37 = Power(Base=FunctionInput(Preview=ScalarParameter() {defaultValue=1, parameterName=Audio_input}) {bUsePreviewValueAsDefault=true, inputName=Audio})
  n38 = LinearInterpolate(Alpha=@n37) {constA=2.6, constB=1.8}
  NamedRerouteDeclaration(None=Saturate(None=Power(Base=@n36, Exp=@n38) {constExponent=20})) {name=SDF, nodeColor=<rgba 0.1016 1 0 1>}
  n39 = Constant() {r=1}
  n40 = Multiply(A=Multiply(A=TextureCoordinate(), B=Constant2Vector() {g=1, r=1}), B=@n39) {constB=3}
  n41 = Multiply(A=@n39) {constB=-0.5}
  n42 = Panner(Coordinate=MaterialFunctionCall(Vector or UVs=Add(A=@n40, B=@n41) {constB=-1.5}) {materialFunction=&VectorToRadialValue}.Vector Converted to Angle) {speedX=0.025, speedY=-0.025}
  n43 = TextureCoordinate()
  n44 = DDX(Value=@n43)
  n45 = DDY(Value=@n43)
  n46 = Power(Base=Multiply(A=TextureSample!(UVs=@n42, Tex, "DDX(UVs)"=@n44, "DDY(UVs)"=@n45, Apply View MipBias) {mipValueMode=TMVM_Derivative, texture=/Clyde/TechArt/Assets/mesh/SkyBox/T_KlinAtmosphericCloudNoise01}.R) {constB=2})
  NamedRerouteDeclaration(None=Saturate(None=LinearInterpolate(Alpha=@n46) {constA=0.3})) {name=EdgeGlow, nodeColor=<rgba 0.4531 0 1 1>}
  n47 = Constant3Vector() {constant=<rgba 0 0.8746 5 1>}
  n48 = Constant3Vector() {constant=<rgba 0 4.0654 15 1>}
  n49 = NamedRerouteUsage()
  NamedRerouteDeclaration(None=LinearInterpolate(B=LinearInterpolate(A=@n47, B=@n48, Alpha=@n49) {constAlpha=0.35}, Alpha=NamedRerouteUsage())) {name=GlowColor, nodeColor=<rgba 0 0.9297 1 1>}
  n50 = Power(Base=@n26.R)
  n51 = LinearInterpolate(Alpha=@n37) {constA=0.02, constB=15}
  n52 = Multiply(A=Multiply(A=Multiply(A=@n50, B=@n51)) {constB=5}, B=@n15) {constB=10}
  NamedRerouteDeclaration(None=@n52) {name=EDGE, nodeColor=<rgba 0 1 0.7422 1>}
  n53 = Constant() {r=1}
  n54 = Saturate(None=Add(A=NamedRerouteUsage(), B=@n52))
  NamedRerouteDeclaration(None=Saturate(None=Multiply(A=@n53, B=@n54) {constA=1})) {name=Opacity_0, nodeColor=<rgba 1 0 0.9375 1>}
  n55 = TextureCoordinate()
  n56 = Constant2Vector() {g=1, r=1}
  n57 = Constant() {r=2}
  n58 = Add(A=Multiply(A=Multiply(A=@n55, B=@n56), B=@n57) {constB=3}, B=Multiply(A=@n57) {constB=-0.5}) {constB=-1.5}
  n59 = TextureSample!(
   UVs=Panner(Coordinate=MaterialFunctionCall(Vector or UVs=@n58) {materialFunction=&VectorToRadialValue}.Vector Converted to Angle) {speedY=-0.01}
   Tex
   Apply View MipBias
  ) {texture=/Clyde/TechArt/Assets/mesh/SkyBox/T_Klin_Stars}
  NamedRerouteDeclaration(None=LinearInterpolate(B=Multiply(A=@n59.R) {constB=30}, Alpha=NamedRerouteUsage())) {name=LightRay, nodeColor=<rgba 1 0 0.8906 1>}
  NamedRerouteDeclaration(None=Constant()) {name=Depth, nodeColor=<rgba 0 1 0.2969 1>}
  NamedRerouteDeclaration(None=Subtract(A=NamedRerouteDeclaration(None=@n21) {name=ScreenMask2, nodeColor=<rgba 0.7578 1 0 1>}, B=NamedRerouteUsage())) {name=ScreenMask3, nodeColor=<rgba 0.8281 1 0 1>}
  n60 = Multiply(A=@n35.R) {constB=3}
  n61 = Subtract(
   A=Abs(None=Subtract(A=@n60))
   B=AppendVector(A=ScalarParameter() {defaultValue=-1.2376, parameterName=Paramferera}, B=ScalarParameter() {defaultValue=-1.5245, parameterName=raadad})
  )
  n62 = Length(None=Max(A=@n61) {constB=0}) {is=class MaterialExpressionLength}
  n63 = Max(A=ComponentMask(None=@n61) {r=true}, B=ComponentMask(None=@n61) {g=true})
  n64 = Power(Base=OneMinus(None=Add(A=@n62, B=@n63)), Exp=ScalarParameter() {defaultValue=1, parameterName=Paramvv}) {constExponent=20}
  OneMinus(None=Multiply(A=Saturate(None=@n64)) {constB=0.5})
  ScalarParameter() {defaultValue=2.5, parameterName=Paramfgq}
 ]}
