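# Translucent screen material for the KilnClyde special event, showing a media video with staged glow, graffiti and shockwave effects.
# Parameters and a collection switch gate the video, which feeds two skyline edge-glow functions whose emissive and opacity are blended.
# A sphere mask and scaled textures build the shockwave and graffiti layers, alongside desaturated, tinted glow blends.
# Named reroutes carry these layers, audio-amplitude reactivity and a normal-plus-pivot world position offset.
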Material /Clyde/Blueprints/Screen/M_SpecialEventsVideo_KilnClyde
graph {
 properties {bAutomaticallySetUsageInEditor=false, blendMode=BLEND_Translucent, usedShadingModels=Default Lit}
 nodes [
  n1 = MaterialFunctionCall!(Direction reversal=Constant() {r=1}, Screen_Scale) {materialFunction=/Clyde/TechArt/Screen/MF_Graffiti_screen_KilnClydePhase1}
  n2 = Multiply(A=NamedRerouteUsage()) {constB=0.1}
  n3 = Multiply(
   A=Desaturation(None=NamedRerouteUsage(), Fraction=Constant() {r=1})
   B=VectorParameter() {parameterName=Screen color after Event, defaultValue=<rgba 0.481 0.6478 1.3 1>}.RGB
  )
  n4 = ScalarParameter() {parameterName=White glow}
  n5 = NamedRerouteUsage()
  n6 = VectorParameter() {parameterName=White Glow Color, defaultValue=<rgba 100 100 100 1>}
  n7 = ScalarParameter() {parameterName=ScreenGlow}
  LinearInterpolate(
   A=LinearInterpolate(A=Multiply(A=@n1.Base color) {constB=0.8}, B=LinearInterpolate(A=@n2, B=@n3, Alpha=@n4), Alpha=@n5)
   B=@n6.RGB
   Alpha=@n7
  )
  LinearInterpolate(A=LinearInterpolate(A=@n1.GraffitiEmissiveColor, B=NamedRerouteUsage(), Alpha=@n5) {constAlpha=1}, B=@n6.RGB, Alpha=@n7)
  n8 = NamedRerouteUsage()
  n9 = TextureSample(UVs=MaterialFunctionCall(UVs=TextureCoordinate() {coordinateIndex=2}, Texture Scale=ScalarParameter() {defaultValue=1, parameterName=ScaleUV}) {materialFunction=&ScaleUVsByCenter}.UVs) {constCoordinate=1, samplerSource=SSM_Wrap_WorldGroupSettings, texture=&/SpecialEventsMediaTexture_Video_KilnClyde}
  n10 = LinearInterpolate(
   A=LinearInterpolate(B=@n9.RGB, Alpha=CollectionParameter() {collection=/Clyde/TechArt/MPC_KilnClyde, parameterName=RR_Screen_Switch})
   Alpha=ScalarParameter() {parameterName=White glow}
  )
  n11 = ScalarParameter() {parameterName=PIPSwitch}
  n12 = LinearInterpolate(
   A=NamedRerouteUsage()
   B=LinearInterpolate(A=LinearInterpolate(A=@n10, Alpha=@n11) {constB=0}, B=NamedRerouteUsage(), Alpha=ScalarParameter() {parameterName=finalStage})
   Alpha=NamedRerouteUsage()
  ) {constAlpha=1}
  n13 = MaterialFunctionCall!(Video switch=@n8, Audio=NamedRerouteUsage(), BaseColor=@n12) {materialFunction=/Clyde/TechArt/Assets/Screen/MF_EdgeGlow_streamScreen_skyline}
  Multiply(A=@n13.EmissiveColor)
  n14 = NamedRerouteUsage()
  n15 = NamedRerouteUsage()
  n16 = SphereMask(
   A=Add(A=TextureCoordinate() {coordinateIndex=1}, B=Multiply(A=@n15) {constB=0.1})
   B=Constant() {r=0.5}
   Radius=ScalarParameter() {parameterName=Shockwav_Radiues}
  ) {hardnessPercent=50}
  n17 = ScalarParameter() {defaultValue=1, parameterName=Video switch}
  Multiply(A=LinearInterpolate(A=Max(A=@n14, B=@n16), Alpha=ScalarParameter() {parameterName=Video switch Protection}), B=@n17)
  n18 = VertexNormalWS()
  n19 = ScalarParameter() {defaultValue=1, parameterName=PositionOffset}
  n20 = Subtract(
   A=WorldPosition().XYZ
   B=MaterialFunctionCall() {materialFunction=&ObjectPivotPoint}.Object Pivot Location
  )
  n21 = ScalarParameter() {defaultValue=0.123, parameterName=Scale}
  NamedRerouteDeclaration(None=Add(A=Multiply(A=@n18, B=@n19) {constB=10}, B=Multiply(A=@n20, B=@n21))) {name=WPO, nodeColor=<rgba 0 1 0.8828 1>}
  n22 = TextureCoordinate() {coordinateIndex=1}
  n23 = ScalarParameter() {defaultValue=3, parameterName=Shockwav_Scale}
  NamedRerouteDeclaration(None=TextureSample(UVs=MaterialFunctionCall(UVs=@n22, Texture Scale=@n23) {materialFunction=&ScaleUVsByCenter}.UVs) {constCoordinate=1, samplerSource=SSM_Wrap_WorldGroupSettings, texture=/Clyde/TechArt/Assets/Screen/T_Shockwave02_TX}.R) {name=ShocwaveSpeedLine, nodeColor=<rgba 1 0.9609 0 1>}
  NamedRerouteDeclaration(None=LinearInterpolate(A=NamedRerouteUsage(), B=@n3, Alpha=ScalarParameter() {parameterName=White glow})) {name=Emissive Color, nodeColor=<rgba 0.3594 0 1 1>}
  n24 = MaterialFunctionCall!(Video switch=@n8, Depth_Weight=ScalarParameter() {parameterName=Depth_weight}, BaseColor=@n12) {materialFunction=/Clyde/TechArt/Assets/Screen/MF_EdgeGlow_streamScreen_SkylineStage2}
  n25 = NamedRerouteUsage()
  NamedRerouteDeclaration(None=LinearInterpolate(A=@n24.Opacity, B=@n13.Opacity, Alpha=@n25)) {name=Opacity, nodeColor=<rgba 0 1 0.0391 1>}
  NamedRerouteDeclaration(None=@n17) {name=Video switch, nodeColor=<rgba 1 0.2578 0 1>}
  n26 = ScalarParameter() {defaultValue=1, parameterName=Video switch}
  n27 = Multiply(A=NamedRerouteUsage(), B=Constant3Vector() {constant=<rgba 0 61.4558 100 1>})
  n28 = TextureCoordinate() {coordinateIndex=1}
  n29 = ScalarParameter() {defaultValue=1, parameterName=ScaleUV}
  n30 = NamedRerouteDeclaration(None=TextureSampleParameter2D(UVs=MaterialFunctionCall(UVs=@n28, Texture Scale=@n29) {materialFunction=&ScaleUVsByCenter}.UVs) {parameterName=Graffiti, samplerSource=SSM_Wrap_WorldGroupSettings, texture=&T_BlacksRGB}.RGB) {name=graffiti, nodeColor=<rgba 1 0 0.9375 1>}
  n31 = Constant3Vector() {constant=<rgba 0.5417 0.5417 0.5417 0>}
  n32 = ScalarParameter() {parameterName=finalStage}
  NamedRerouteDeclaration(None=Add(A=Multiply(A=@n26, B=@n27), B=LinearInterpolate(A=@n30, B=@n31, Alpha=@n32))) {name=shockwaveGlow, nodeColor=<rgba 0 0.0625 1 1>}
  NamedRerouteDeclaration(None=LinearInterpolate(A=@n24.EmissiveColor, B=@n13.EmissiveColor, Alpha=@n25)) {name=EmissiveColorOutput, nodeColor=<rgba 0.3828 1 0 1>}
  NamedRerouteDeclaration(None=@n16) {name=SphereMask Opacity, nodeColor=<rgba 1 0.2812 0 1>}
  n33 = Power(Base=MaterialFunctionCall() {materialFunction=/Game/Packages/Fortress_SharedMaterials/Base_Material_Functions/BaseCharacterShaderFunctions/Audio/MF_AudioReactivity}.Amplitude) {constExponent=10}
  NamedRerouteDeclaration(None=Saturate(None=Multiply(A=@n33) {constB=5})) {name=AudioReactivity, nodeColor=<rgba 0 0.1094 1 1>}
  NamedRerouteUsage()
  NamedRerouteUsage()
 ]
}
