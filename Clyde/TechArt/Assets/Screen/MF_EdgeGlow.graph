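# MF_EdgeGlow is a material function that adds an animated glowing border and light rays to a screen surface, outputting EmissiveColor and Opacity.
# Two rounded-box SDFs on UV1, warped by a panning flow texture, combine with HeightLerp masks to form edge and depth masks.
# Radially mapped noise and star textures drive glow and rays, with explicit derivatives on some samples.
# The Video switch input scales the warp and blends the output opacity.
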MaterialFunction /Clyde/TechArt/Assets/Screen/MF_EdgeGlow
graph {nodes [
  n1 = Multiply(
   A=NamedRerouteDeclaration(None=NamedRerouteUsage()) {name=Opacity_1, nodeColor=<rgba 1 0 0.2812 1>}
   B=Add(A=NamedRerouteUsage(), B=NamedRerouteUsage())
  )
  n2 = FunctionInput(Preview=Constant3Vector()) {inputName=BaseColor}
  n3 = Saturate(None=NamedRerouteUsage())
  FunctionOutput(None=NamedRerouteDeclaration(None=LinearInterpolate(A=@n1, B=@n2, Alpha=@n3)) {name=EmissiveColor, nodeColor=<rgba 1 0.6328 0 1>}) {outputName=EmissiveColor}
  n4 = NamedRerouteUsage()
  n5 = NamedRerouteUsage()
  n6 = Power(Base=NamedRerouteUsage())
  n7 = Max(A=@n4, B=Max(A=Power(Base=@n5), B=Saturate(None=@n6)))
  n8 = NamedRerouteDeclaration(None=FunctionInput() {bUsePreviewValueAsDefault=true, inputName=Video switch, inputType=FunctionInput_Scalar}) {name=VideoSwitch, nodeColor=<rgba 0.2656 0 1 1>}
  FunctionOutput(None=LinearInterpolate(A=@n4, B=NamedRerouteDeclaration(None=@n7) {name=Opacity, nodeColor=<rgba 0 1 0.9531 1>}, Alpha=@n8)) {outputName=Opacity}
  n9 = NamedRerouteUsage()
  n10 = Constant()
  n11 = Constant() {r=1}
  n12 = Constant() {r=0.1}
  n13 = NamedRerouteUsage()
  n14 = Constant()
  n15 = Saturate(None=Multiply(
    A=MaterialFunctionCall(A=@n10, B=@n11, Transition Phase=@n12, Height Texture=@n13, Contrast=@n14) {materialFunction=&HeightLerp}.Results
    B=Constant() {r=30}
   ))
  n16 = Subtract(A=NamedRerouteUsage(), B=@n15)
  n17 = Max(A=NamedRerouteUsage(), B=@n16)
  NamedRerouteDeclaration(None=Add(A=NamedRerouteUsage(), B=Multiply(A=@n9, B=@n17))) {name=EdgeMask, nodeColor=<rgba 1 0 0.5156 1>}
  n18 = TextureCoordinate() {coordinateIndex=1}
  n19 = Constant2Vector() {g=1, r=1}
  n20 = Constant() {r=2}
  n21 = Add(A=Multiply(A=Multiply(A=@n18, B=@n19), B=@n20) {constB=3}, B=Multiply(A=@n20) {constB=-0.5}) {constB=-1.5}
  n22 = TextureSample(UVs=Panner(Coordinate=MaterialFunctionCall(Vector or UVs=@n21) {materialFunction=&VectorToRadialValue}.Vector Converted to Angle) {speedY=-0.01}) {texture=/Clyde/TechArt/Assets/mesh/SkyBox/T_Klin_Stars}
  NamedRerouteDeclaration(None=LinearInterpolate(B=Multiply(A=@n22.R) {constB=30}, Alpha=NamedRerouteUsage())) {name=LightRay, nodeColor=<rgba 1 0 0.8906 1>}
  n23 = NamedRerouteUsage()
  n24 = TextureCoordinate() {coordinateIndex=1}
  n25 = Constant2Vector() {g=1, r=1}
  n26 = Constant() {r=1}
  n27 = Add(A=Multiply(A=Multiply(A=@n24, B=@n25), B=@n26) {constB=3}, B=Multiply(A=@n26) {constB=-0.5}) {constB=-1.5}
  n28 = TextureCoordinate() {coordinateIndex=1}
  n29 = TextureSample!(
   UVs=Panner(Coordinate=MaterialFunctionCall(Vector or UVs=@n27) {materialFunction=&VectorToRadialValue}.Radial Coordinates) {speedX=0.1, speedY=-0.02}
   Tex
   "DDX(UVs)"=DDX(Value=@n28)
   "DDY(UVs)"=DDY(Value=@n28)
   Apply View MipBias
  ) {
   mipValueMode=TMVM_Derivative
   samplerSource=SSM_Wrap_WorldGroupSettings
   samplerType=SAMPLERTYPE_Normal
   texture=/Clyde/TechArt/Assets/mesh/SkyBox/T_Klin_Flow_03
  }
  n30 = LinearInterpolate(B=Constant() {r=0.02}, Alpha=@n8)
  n31 = MaterialFunctionCall(
   UVs=Add(A=TextureCoordinate() {coordinateIndex=1}, B=Multiply(A=@n29.R, B=@n30) {constB=0.05})
   Texture Scale=ScalarParameter() {defaultValue=1.75, parameterName=Square uvScale}
  ) {materialFunction=&ScaleUVsByCenter}
  n32 = Multiply(A=@n31.UVs) {constB=2}
  n33 = Subtract(
   A=Abs(None=Subtract(A=@n32))
   B=AppendVector(
    A=ScalarParameter() {defaultValue=0.495, parameterName=Screen Box width2}
    B=ScalarParameter() {defaultValue=0.495, parameterName=Height of box2}
   )
  )
  n34 = Length(None=Max(A=@n33) {constB=0}) {is=class MaterialExpressionLength}
  n35 = Max(A=ComponentMask(None=@n33) {r=true}, B=ComponentMask(None=@n33) {g=true})
  n36 = Power(Base=OneMinus(None=Add(A=@n34, B=@n35)), Exp=Constant() {r=8}) {constExponent=20}
  n37 = NamedRerouteDeclaration(None=OneMinus(None=Saturate(None=@n36))) {name=SDF_1, nodeColor=<rgba 0.1016 1 0 1>}
  n38 = FunctionInput() {bUsePreviewValueAsDefault=true, inputName=Depth_Weight, inputType=FunctionInput_Scalar, previewValue=<1 0 0 1>}
  NamedRerouteDeclaration(None=Multiply(A=Multiply(A=@n23, B=@n37), B=@n38)) {name=Depth, nodeColor=<rgba 0 1 0.2969 1>}
  NamedRerouteDeclaration(None=Subtract(A=NamedRerouteDeclaration(None=@n15) {name=ScreenMask2, nodeColor=<rgba 0.7578 1 0 1>}, B=NamedRerouteUsage())) {name=ScreenMask3, nodeColor=<rgba 0.8281 1 0 1>}
  n39 = Multiply(A=@n31.UVs) {constB=2}
  n40 = Subtract(
   A=Abs(None=Subtract(A=@n39))
   B=AppendVector(
    A=ScalarParameter() {defaultValue=0.51, parameterName=Screen Box width}
    B=ScalarParameter() {defaultValue=0.509, parameterName=Height of box}
   )
  )
  n41 = Max(A=@n40) {constB=0}
  n42 = ComponentMask(None=@n40) {r=true}
  n43 = ComponentMask(None=@n40) {g=true}
  n44 = OneMinus(None=Add(A=Length(None=@n41), B=Max(A=@n42, B=@n43)))
  n45 = Constant() {r=18}
  NamedRerouteDeclaration(None=Saturate(None=Power(Base=@n44, Exp=@n45) {constExponent=20})) {name=SDF, nodeColor=<rgba 0.1016 1 0 1>}
  n46 = Constant() {r=1}
  n47 = Multiply(A=Multiply(A=TextureCoordinate() {coordinateIndex=1}, B=Constant2Vector() {g=1, r=1}), B=@n46) {constB=3}
  n48 = Multiply(A=@n46) {constB=-0.5}
  n49 = Panner(Coordinate=MaterialFunctionCall(Vector or UVs=Add(A=@n47, B=@n48) {constB=-1.5}) {materialFunction=&VectorToRadialValue}.Vector Converted to Angle) {speedX=0.025, speedY=-0.025}
  n50 = TextureCoordinate()
  n51 = DDX(Value=@n50)
  n52 = DDY(Value=@n50)
  n53 = Power(Base=Multiply(A=TextureSample(UVs=@n49, "DDX(UVs)"=@n51, "DDY(UVs)"=@n52) {mipValueMode=TMVM_Derivative, texture=/Clyde/TechArt/Assets/mesh/SkyBox/T_KlinAtmosphericCloudNoise01}.R) {constB=2})
  NamedRerouteDeclaration(None=Saturate(None=LinearInterpolate(Alpha=@n53) {constA=0.3})) {name=EdgeGlow, nodeColor=<rgba 0.4531 0 1 1>}
  n54 = Constant3Vector() {constant=<rgba 0 1.7492 10 1>}
  n55 = Constant3Vector() {constant=<rgba 0 5.4206 20 1>}
  n56 = NamedRerouteUsage()
  NamedRerouteDeclaration(None=LinearInterpolate(B=LinearInterpolate(A=@n54, B=@n55, Alpha=@n56) {constAlpha=0.35}, Alpha=NamedRerouteUsage())) {name=GlowColor, nodeColor=<rgba 0 0.9297 1 1>}
  n57 = NamedRerouteUsage()
  n58 = SmoothStep(Min=LinearInterpolate(A=Constant() {r=0.75}, B=Constant() {r=0.65}, Alpha=@n57) {constA=0.8}, Max=Constant() {r=1}, Value=@n16)
  NamedRerouteDeclaration(None=Multiply(A=Multiply(A=@n58, B=@n57)) {constB=60}) {name=EDGE, nodeColor=<rgba 0 1 0.7422 1>}
  n59 = ScalarParameter() {defaultValue=1, parameterName=Screen weight}
  n60 = NamedRerouteUsage()
  NamedRerouteDeclaration(None=Saturate(None=Multiply(A=@n59, B=@n60) {constA=1})) {name=Opacity_0, nodeColor=<rgba 1 0 0.9375 1>}
 ]}
