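# MF_ScreenGlow is a material function for a glowing screen effect, outputting emissive color, base color and opacity.
# A radially panned flow texture distorts UV1 to build a soft rounded-box SDF mask.
# A panning, saturation-boosted rainbow texture blends with the "Screen color after Event" tint to form the color.
# MF_EdgeGlow supplies an opacity reroute, and MF_FX_EmissiveAdjustments processes the masked color for emissive.
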MaterialFunction /Clyde/Blueprints/Screen/MF_ScreenGlow
graph {nodes [
  n1 = Constant() {r=0.25}
  AppendVector(A=@n1, B=Constant() {r=0.45})
  n2 = NamedRerouteUsage()
  n3 = TextureCoordinate() {coordinateIndex=1}
  n4 = Multiply(A=NamedRerouteUsage()) {constB=0.2}
  n5 = TextureSample(UVs=Panner(Coordinate=Add(A=@n3, B=@n4)) {speedY=0.1}) {constCoordinate=1, texture=/Clyde/TechArt/Assets/mesh/SkyBox/T_Klin_RainbowColor}
  n6 = Desaturation(None=@n5.RGB, Fraction=Constant() {r=-0.3})
  n7 = VectorParameter() {parameterName=Screen color after Event, defaultValue=<rgba 0.7514 1.465 2 1>}
  n8 = Add(
   A=Multiply(A=Multiply(A=@n2, B=@n6)) {constB=20}
   B=LinearInterpolate(A=@n7.RGB, B=Multiply(A=@n5.RGB, B=@n7.RGB), Alpha=NamedRerouteUsage())
  )
  n9 = NamedRerouteUsage()
  FunctionOutput(None=MaterialFunctionCall(Emissive=Multiply(A=@n8, B=@n9)) {materialFunction=&MF_FX_EmissiveAdjustments}.Result) {outputName=Emissive Color, sortPriority=1}
  FunctionOutput(None=NamedRerouteDeclaration(None=@n8) {name=Base Color, nodeColor=<rgba 0 0.4375 1 1>}) {outputName=BaseColor}
  FunctionOutput(None=@n9) {outputName=Opacity, sortPriority=2}
  Multiply(A=@n8, B=TextureSample() {samplerType=SAMPLERTYPE_Normal, texture=/Clyde/TechArt/Screen/Texture/T_Recreation_MovieScreen_A_C_N}.B)
  NamedRerouteDeclaration(None=MaterialFunctionCall!(Video switch=Constant(), Depth_Weight=Constant() {r=1}, BaseColor) {materialFunction=/Clyde/TechArt/Assets/Screen/MF_EdgeGlow}.Opacity) {name=Opacity, nodeColor=<rgba 0 1 0.0391 1>}
  n10 = TextureCoordinate() {coordinateIndex=1}
  n11 = TextureCoordinate() {coordinateIndex=1}
  n12 = Constant2Vector() {g=1, r=1}
  n13 = Constant() {r=1}
  n14 = Add(A=Multiply(A=Multiply(A=@n11, B=@n12), B=@n13) {constB=3}, B=Multiply(A=@n13) {constB=-0.5}) {constB=-1.5}
  n15 = TextureCoordinate() {coordinateIndex=1}
  n16 = TextureSample!(
   UVs=Panner(Coordinate=MaterialFunctionCall(Vector or UVs=@n14) {materialFunction=&VectorToRadialValue}.Radial Coordinates) {speedX=0.01, speedY=-0.1}
   Tex
   "DDX(UVs)"=DDX(Value=@n15)
   "DDY(UVs)"=DDY(Value=@n15)
   Apply View MipBias
  ) {
   mipValueMode=TMVM_Derivative
   samplerSource=SSM_Wrap_WorldGroupSettings
   samplerType=SAMPLERTYPE_Normal
   texture=/Clyde/TechArt/Assets/mesh/SkyBox/T_Klin_Flow_03
  }
  n17 = Multiply(A=NamedRerouteDeclaration(None=@n16.R) {name="UV  Noise", nodeColor=<rgba 1 0 0.1406 1>}, B=Constant() {r=0.1}) {constB=0.05}
  n18 = Multiply(A=MaterialFunctionCall(UVs=Add(A=@n10, B=@n17), Texture Scale=Constant() {r=1.75}) {materialFunction=&ScaleUVsByCenter}.UVs) {constB=2}
  n19 = Subtract(A=Abs(None=Subtract(A=@n18)), B=@n1)
  n20 = Add(
   A=Length(None=Max(A=@n19) {constB=0}) {is=class MaterialExpressionLength}
   B=Max(A=ComponentMask(None=@n19) {r=true}, B=ComponentMask(None=@n19) {g=true})
  )
  n21 = Power(Base=@n20, Exp=Constant() {r=3}) {constExponent=20}
  NamedRerouteDeclaration(None=Saturate(None=@n21)) {name=SDF2, nodeColor=<rgba 0.1016 1 0 1>}
  OneMinus(None=@n20)
  OneMinus(None=Saturate(None=@n21))
 ]}
