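# Translucent lit material for the floating screen that shows the Kiln Clyde special event video.
# Audio amplitude is sharpened into a pulse that pushes the video's red and blue samples apart horizontally for a beat-reactive RGB split.
# An MPC_KilnClyde switch blends in the split video, which feeds an edge-glow function producing emissive and opacity.
# Parameters blend in a desaturated post-event tint and screen glow, and control picture-in-picture and opacity.
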Material /Clyde/Blueprints/Screen/M_SpecialEventsVideo_KilnClyde_floatScreen
graph {
 properties {bAutomaticallySetUsageInEditor=false, blendMode=BLEND_Translucent, usedShadingModels=Default Lit}
 nodes [
  n1 = NamedRerouteUsage()
  LinearInterpolate(Alpha=@n1)
  n2 = NamedRerouteUsage()
  n3 = Multiply(
   A=Desaturation(None=NamedRerouteUsage(), Fraction=Constant() {r=1})
   B=VectorParameter() {parameterName=Screen color after Event, defaultValue=<rgba 0.481 0.6478 1.3 1>}.RGB
  )
  n4 = MaterialFunctionCall() {materialFunction=&/MF_ScreenGlow}
  n5 = ScalarParameter() {parameterName=ScreenGlow}
  LinearInterpolate(A=LinearInterpolate(A=Multiply(A=@n2) {constB=0.1}, B=@n3, Alpha=ScalarParameter() {parameterName=White glow}), B=@n4.BaseColor, Alpha=@n5)
  LinearInterpolate(A=NamedRerouteUsage(), B=@n4.Emissive Color, Alpha=@n5)
  n6 = Power(Base=MaterialFunctionCall() {materialFunction=/Game/Packages/Fortress_SharedMaterials/Base_Material_Functions/BaseCharacterShaderFunctions/Audio/MF_AudioReactivity}.Amplitude) {constExponent=10}
  Multiply(A=@n6) {constB=2}
  Multiply(A=@n1) {constB=1.1}
  n7 = NamedRerouteUsage()
  Multiply(A=Multiply(A=LinearInterpolate(A=@n7, B=@n4.Opacity, Alpha=@n5), B=ScalarParameter() {parameterName=FloatScreen_Opacity})) {constB=0.5}
  n8 = Saturate(None=Multiply(A=@n6) {constB=5})
  NamedRerouteDeclaration(None=@n8) {name=AudioReactivity, nodeColor=<rgba 0 0.1094 1 1>}
  n9 = NamedRerouteUsage()
  n10 = ScalarParameter() {parameterName=White glow}
  NamedRerouteDeclaration(None=MaterialFunctionCall(Emissive=LinearInterpolate(A=@n9, B=@n3, Alpha=@n10)) {materialFunction=&MF_FX_EmissiveAdjustments}.Result) {name=Emissive Color, nodeColor=<rgba 0.3594 0 1 1>}
  n11 = TextureCoordinate()
  n12 = Constant2Vector() {r=1}
  n13 = Multiply(A=@n8) {constB=0.25}
  n14 = TextureSample(UVs=Add(A=@n11, B=Multiply(A=@n12, B=@n13))) {constCoordinate=1, samplerSource=SSM_Wrap_WorldGroupSettings, texture=&/SpecialEventsMediaTexture_Video_KilnClyde}
  n15 = Multiply(A=@n14.R)
  n16 = TextureSample(UVs=@n11) {constCoordinate=1, samplerSource=SSM_Wrap_WorldGroupSettings, texture=&SpecialEventsMediaTexture_Video_KilnClyde}
  n17 = Multiply(A=@n16.G)
  n18 = Constant2Vector() {r=-1}
  n19 = TextureSample(UVs=Add(A=@n11, B=Multiply(A=@n13, B=@n18))) {constCoordinate=1, samplerSource=SSM_Wrap_WorldGroupSettings, texture=&SpecialEventsMediaTexture_Video_KilnClyde}
  n20 = LinearInterpolate(
   A=NamedRerouteUsage()
   B=AppendVector(A=AppendVector(A=@n15, B=@n17), B=Multiply(A=@n19.B))
   Alpha=CollectionParameter() {collection=/Clyde/TechArt/MPC_KilnClyde, parameterName=RR_Screen_Switch}
  )
  n21 = ScalarParameter() {parameterName=White glow}
  n22 = MaterialFunctionCall!(
   Video switch=NamedRerouteUsage()
   Depth_Weight=ScalarParameter() {defaultValue=1, parameterName=Depth_weight}
   Screen_Scale_Input=LinearInterpolate(Alpha=@n1) {constA=0.94, constB=1.07}
   BaseColor=LinearInterpolate(A=LinearInterpolate(A=@n20, Alpha=@n21), Alpha=ScalarParameter() {parameterName=PIPSwitch}) {constB=0}
  ) {materialFunction=/Clyde/TechArt/Assets/Screen/MF_EdgeGlow_floatScreen}
  NamedRerouteDeclaration(None=@n22.Opacity) {name=Opacity, nodeColor=<rgba 0 1 0.0391 1>}
  NamedRerouteDeclaration(None=ScalarParameter() {defaultValue=1, parameterName=Video switch}) {name=Video switch, nodeColor=<rgba 1 0.2578 0 1>}
  NamedRerouteDeclaration(None=@n22.EmissiveColor) {name=EmissiveColorOutput, nodeColor=<rgba 0.3828 1 0 1>}
  n23 = TextureCoordinate()
  n24 = ScalarParameter() {defaultValue=1, parameterName=ScaleUV}
  NamedRerouteDeclaration(None=TextureSampleParameter2D(UVs=MaterialFunctionCall(UVs=@n23, Texture Scale=@n24) {materialFunction=&ScaleUVsByCenter}.UVs) {parameterName=Graffiti, samplerSource=SSM_Wrap_WorldGroupSettings, texture=&T_BlacksRGB}.RGB) {name=graffiti, nodeColor=<rgba 1 0 0.9375 1>}
  TextureSample(UVs=MaterialFunctionCall(UVs=TextureCoordinate(), Texture Scale=ScalarParameter() {defaultValue=1, parameterName=ScaleUV}) {materialFunction=&ScaleUVsByCenter}.UVs) {constCoordinate=1, samplerSource=SSM_Wrap_WorldGroupSettings, texture=&SpecialEventsMediaTexture_Video_KilnClyde}
 ]
}
