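# Niagara effect for a camera screen projector, drawing a flickering, glowing sprite tinted by User.MasterColorAndOpacity.
# Its single Main emitter burst-spawns on a timed loop, with particle lifetime tied to loop duration and sprites shifted by a camera offset.
# Nested waveforms drive uniform sprite scale for an irregular high-frequency flicker.
# A boosted copy of particle color is written to a custom Base glowColor attribute.
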NiagaraSystem /Clyde/Blueprints/Screen/NS_CameraScreenProjectorFX
dependencies {
 usedDataInterfaces [&NiagaraDataInterfaceCurve]
 usedDynamicInputs [&Multiply_Vector]
 usedModules [&SystemState]
}
emitters {Main {
  emitterUpdateScript [
   &EmitterState(Life Cycle Mode=::NewEnumerator1, Loop Duration=4, UseLoopDelay=false)
   &SpawnBurst_Instantaneous(Use Spawn Probability=false, UseLoopCountLimit=false)
  ]
  particleSpawnScript [
   &InitializeParticle(
    Lifetime=$Emitter.CurrentLoopDuration
    Sprite Size=&Multiply_Vector2D(A=<100 100>, B=&Multiply_Vector2D_ByFloat(Float=5, Vector2D=<1 1>))
    Sprite Size Mode=::NewEnumerator4
    UsePositionOffset=false
    Write Material Random=true
   )
   &CameraOffset(Camera Offset Amount=25)
  ]
  particleUpdateScript [
   &ParticleState()
   &ScaleSpriteSize(
    Scale Sprite Size Mode=::NewEnumerator0
    Uniform Scale Factor=&Multiply_Float(
     A=&Waveform(
      "Waveform [1]"=::NewEnumerator12
      "[1] Phase Scale"=&Waveform("Waveform [1]"=::NewEnumerator12, "[1] Input Position"=16789.3105, "[1] Frequency"=1, "[1] Amplitude Min/Max"=<0.05 1>)
      "[1] Frequency"=50
      "[1] Amplitude Min/Max"=<0.85 1>
     )
     B=1
    )
   )
   &Color(Color=$User.MasterColorAndOpacity)
   /SetVariables(Particles.Base glowColor:LinearColor=&MultiplyLinearColor(
     Linear Color=$Particles.Color
     Scale Factor=&MakeFloat4FromVectorAndFloat("Vector (RGB)"=&VectorFromFloat(Value=7), "Float (Alpha)"=2)
    )) {isSetParameters=true}
  ]
  rendererClasses [&NiagaraSpriteRendererProperties]
  renderers [{rendererClass=&NiagaraSpriteRendererProperties}]
 }}
summary {
 systemName=NS_CameraScreenProjectorFX
 emitters [{emitterName=Main, rendererClasses [&NiagaraSpriteRendererProperties]}]
 userVariables [User.MasterColorAndOpacity {type=LinearColor, description="", defaultValue {struct=LinearColor, value=<rgba 1 1 1 1>}}]
}
